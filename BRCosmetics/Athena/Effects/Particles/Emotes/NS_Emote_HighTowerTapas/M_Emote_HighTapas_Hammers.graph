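# Unlit translucent Niagara mesh-particle material for hammer blur meshes in the HighTowerTapas emote effect.
# A blue Color parameter times particle color feeds the shared FX emissive adjustment function.
# An alpha mask sums vertex color RGB channels, each weighted by a dynamic parameter so Niagara can blend them.
# That mask is multiplied by particle alpha, vertex alpha and an inverted Fresnel that fades glancing edges.
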Material /BRCosmetics/Athena/Effects/Particles/Emotes/NS_Emote_HighTowerTapas/M_Emote_HighTapas_Hammers
graph {
 properties {
  bUsedWithNiagaraMeshParticles=true
  blendMode=BLEND_Translucent
  pixelDepthOffsetMode=PDOM_Legacy
  previewMesh=&/SM_Emote_HighTapas_Blur_Hammer
  shadingModel=MSM_Unlit
  usedShadingModels=Unlit
 }
 nodes [
  n1 = ParticleColor()
  MaterialFunctionCall(Emissive=Multiply(A=VectorParameter() {parameterName=Color, defaultValue=<rgba 0.1198 0.6406 1 1>}.RGB, B=@n1.RGB)) {materialFunction=&MF_FX_EmissiveAdjustments}
  n2 = VertexColor()
  n3 = DynamicParameter() {paramNames [VertexColorR,VertexColorG,VertexColorB,VertexColorA]}
  n4 = Add(A=Multiply(A=@n2.R, B=@n3.VertexColorR), B=Multiply(A=@n2.G, B=@n3.VertexColorG))
  n5 = Multiply(A=@n2.B, B=@n3.VertexColorB)
  n6 = OneMinus(None=Fresnel() {baseReflectFraction=0, exponent=0.05})
  Multiply(A=@n1.A, B=Multiply(A=Add(A=@n4, B=@n5), B=Multiply(A=@n2.A, B=@n6)))
 ]
}
