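# Niagara system for the HighTower Tapas emote that plays blue lightning, flares, sparks and flickering light over the character and a weapon.
# Skeletal mesh sampling places flares and lightning meshes on User.Weapon, and smaller bolts, sparks and pulsing lit sprites on body bones and hand sockets.
# Hand arc sprites align between two sampled bone positions.
# Short, staggered bursts sequence the strike, with curves driving color, fades and dynamic material parameters.
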NiagaraSystem /BRCosmetics/Athena/Effects/Particles/Emotes/NS_Emote_HighTowerTapas/NS_Hightower_Tapas_Emote
dependencies {
 usedDataInterfaces [
  &NiagaraDataInterfaceCurve
  &NiagaraDataInterfaceColorCurve
  &NiagaraDataInterfaceVector2DCurve
  &NiagaraDataInterfaceVectorCurve
 ]
 usedDynamicInputs []
 usedModules [&SystemState]
}
emitters {
 Blue_Decal {
  emitterUpdateScript [
   &EmitterState(Life Cycle Mode=::NewEnumerator1, Loop Behavior=::NewEnumerator1, Loop Duration=1, UseLoopDelay=true)
   &SpawnBurst_Instantaneous(Spawn Count=30, Use Spawn Probability=false)
  ]
  particleSpawnScript [
   &FN_SetSpawnGroup_v2(Number of Spawn Groups=15)
   /Niagara/Modules/Spawn/Initialization/InitializeParticle(
    Lifetime=0.5
    Module.Write Color=true
    Module.Write Lifetime=true
    Module.Write Mass=true
    Module.Write Material Random=true
    Module.Write Position=false
    Module.Write Scale=false
    Module.Write SpriteRotation=false
    Module.Write SpriteSize=true
    Position=&SimulationPosition()
    Sprite Size=&Vector2DFromFloat(Value=80)
   )
  ]
  particleUpdateScript [
   &ParticleState()
   &SkeletalMeshLocation(
    Apply Sampled Orientation to Mesh Particles=true
    Bone / Socket Index=$Particles.SpawnGroup
    Bone Sampling Mode=::NewEnumerator4
    Inherit Velocity=false
    MaskSpawnGroup=false
    Sampled Position Offset=&MakeCustomVectorFromBool(Bool=HlslExpression() {hlslExpression="Particles.UniqueID < 15 ? 1:0"}, True Vector=<0 20 1>, False Vector=<1 -20 0>)
    Skeletal Mesh=DataInterface() {propertyValues {
      DefaultMesh=null
      ExcludeBoneName=root
      OptionalComponentClass=null
      PreviewMesh=&M_Med_Soldier_04_Preview
      SkinningMode=SkinOnTheFly
      SoftSourceActor=null
      SourceMode=Default
      UvSetIndex=0
      WholeMeshLOD=-1
      bExcludeBone=false
      bReadDeformedGeometry=true
      bRequireCurrentFrameData=true
      ComponentTags []
      FilteredBones [head,neck_01,spine_02,spine_05,pelvis,upperarm_r,lowerarm_r,upperarm_l,lowerarm_l,hand_l,hand_r,calf_l,calf_r,thigh_l,thigh_r]
      FilteredSockets []
      MeshUserParameter {parameter {name=None, typeDefHandle {registeredTypeIndex=1}, varData []}}
      SamplingRegions []
     }}
   )
   &SolveForcesAndVelocity(Clamp Velocity=false, Limit Acceleration=false)
   &Color(
    Color=<rgba 0.0645 0.4112 1 1>
    Scale Color=&VectorFromFloat(Value=&FloatFromCurve(FloatCurve=DataInterface() {propertyValues {Curve {keys [
          {interpMode=RCIM_Cubic, tangentMode=RCTM_Break, time=0, value=0}
          {arriveTangent=73.8195, interpMode=RCIM_Cubic, leaveTangent=-74.3959, tangentMode=RCTM_Break, time=0.5, value=12}
          {interpMode=RCIM_Cubic, tangentMode=RCTM_Break, time=1, value=0}
         ]}}}))
   )
   /SetVariables(Particles.LightColor=&MultiplyLinearColor(
     Linear Color=<rgba 0.0348 0.4387 1 1>
     Scale Factor=&MakeVector4_FromFloat(Float=&FloatFromCurve(FloatCurve=DataInterface() {propertyValues {Curve {keys [
           {interpMode=RCIM_Cubic, tangentMode=RCTM_Break, time=0, value=0}
           {arriveTangent=73.7962, interpMode=RCIM_Cubic, leaveTangent=-81.5322, tangentMode=RCTM_Break, time=0.5, value=12}
           {interpMode=RCIM_Cubic, tangentMode=RCTM_Break, time=1, value=0}
          ]}}}))
    )) {isSetParameters=true}
  ]
  rendererClasses [&NiagaraSpriteRendererProperties, &NiagaraLightRendererProperties]
  renderers [
   {rendererClass=&NiagaraSpriteRendererProperties}
   {rendererClass=&NiagaraLightRendererProperties, rendererIndex=1}
  ]
 }
 Flare {
  emitterUpdateScript [
   &EmitterState(Life Cycle Mode=::NewEnumerator1, Loop Behavior=::NewEnumerator1, Loop Duration=0.3, UseLoopDelay=false)
   &SpawnRate(SpawnRate=5, Use Spawn Probability=false)
   &SpawnBurst_Instantaneous(Spawn Count=1, Use Spawn Probability=false)
  ]
  particleSpawnScript [
   &InitializeParticle(
    Color=<rgba 0.3056 3.3183 7 1>
    Lifetime=0.4
    Module.Write Color=true
    Module.Write Lifetime=true
    Module.Write Mass=true
    Module.Write Material Random=true
    Module.Write Position=false
    Module.Write Scale=false
    Module.Write SpriteRotation=false
    Module.Write SpriteSize=true
    Position=&SimulationPosition()
    Sprite Size=<400 600>
   )
   &AddVelocity(Coordinate Space=::World, Velocity=<0 0 0.1>)
   /Niagara/Modules/Spawn/Location/SphereLocation(MaskSpawnGroup=false, Module.Override Randomness=false, Sphere Radius=0, Surface Only=false)
   &SkeletalMeshLocation(
    Inherit Velocity=false
    Interpolate Mesh Sampling=false
    MaskSpawnGroup=false
    Mesh Sampling Type=::NewEnumerator5
    Sampled Orientation Coordinate Space=::World
    Skeletal Mesh=$User.Weapon
    Socket Sampling Mode=::NewEnumerator5
   )
  ]
  particleUpdateScript [
   &ParticleState()
   &SolveForcesAndVelocity(Clamp Velocity=false, Limit Acceleration=false)
   &ScaleColor(
    Scale Alpha=&FloatFromCurve(FloatCurve=DataInterface() {propertyValues {Curve {keys [
         {arriveTangent=-2.7664, interpMode=RCIM_Cubic, leaveTangent=-2.7664, tangentMode=RCTM_User, time=0, value=0.5}
         {arriveTangent=2, interpMode=RCIM_Cubic, leaveTangent=2, tangentMode=RCTM_Break, time=0.2, value=0.2621}
         {arriveTangent=-0.3277, interpMode=RCIM_Cubic, leaveTangent=-3.899, tangentMode=RCTM_Break, time=0.25, value=1}
         <at 1 0>
        ]}}})
    ScaleA=true
    ScaleRGB=true
   )
   &DynamicMaterialParameters(
    Index 0 Param 1=&FloatFromCurve(FloatCurve=DataInterface() {propertyValues {Curve {keys [<at 0.25 0.7>,<at 0.5 0.05>]}}})
    Index 0 Param 2=2
    Index 0 Param 3=1
    Index 0 Param 4=1
    Module.Param0WriteEnabled=true
    Module.Param1WriteEnabled=true
    Module.Param2WriteEnabled=true
    Module.Param3WriteEnabled=true
   )
   &CameraOffset(Camera Offset Amount=50)
  ]
  rendererClasses [&NiagaraSpriteRendererProperties]
  renderers [{rendererClass=&NiagaraSpriteRendererProperties}]
 }
 FlareTop {
  emitterUpdateScript [
   &EmitterState(Life Cycle Mode=::NewEnumerator1, Loop Behavior=::NewEnumerator1, Loop Duration=0.2, UseLoopDelay=false)
   &SpawnRate(SpawnRate=5, Use Spawn Probability=false)
   &SpawnBurst_Instantaneous(Spawn Count=1, Use Spawn Probability=false)
  ]
  particleSpawnScript [
   &InitializeParticle(
    Color=<rgba 0.4761 4.2673 7 1>
    Lifetime=0.4
    Module.Write Color=true
    Module.Write Lifetime=true
    Module.Write Mass=true
    Module.Write Material Random=true
    Module.Write Position=false
    Module.Write Scale=false
    Module.Write SpriteRotation=false
    Module.Write SpriteSize=true
    Position=&SimulationPosition()
    Sprite Size=&Vector2DFromFloat(Value=500)
   )
   &AddVelocity(Velocity=<0 0 0.1>)
   &SphereLocation(MaskSpawnGroup=false, Module.Override Randomness=false, Offset=<0 0 350>, Sphere Radius=0, Surface Only=false) {enabled=false}
  ]
  particleUpdateScript [
   &SkeletalMeshLocation(
    Inherit Velocity=false
    Interpolate Mesh Sampling=false
    MaskSpawnGroup=false
    Mesh Sampling Type=::NewEnumerator5
    Sampled Orientation Coordinate Space=::World
    Sampled Position Offset=<0 0 370>
    Skeletal Mesh=$User.Weapon
    Socket Sampling Mode=::NewEnumerator5
   )
   &ParticleState()
   &SolveForcesAndVelocity(Clamp Velocity=false, Limit Acceleration=false)
   &ScaleColor(
    Scale Alpha=&FloatFromCurve(FloatCurve=DataInterface() {propertyValues {Curve {keys [
         {arriveTangent=-2.7664, interpMode=RCIM_Cubic, leaveTangent=-2.7664, tangentMode=RCTM_User, time=0, value=0.5}
         {arriveTangent=2, interpMode=RCIM_Cubic, leaveTangent=2, tangentMode=RCTM_Break, time=0.2, value=0.2621}
         {arriveTangent=-0.3277, interpMode=RCIM_Cubic, leaveTangent=-3.899, tangentMode=RCTM_Break, time=0.25, value=1}
         <at 1 0>
        ]}}})
    ScaleA=true
    ScaleRGB=true
   )
   &DynamicMaterialParameters(
    Index 0 Param 1=&FloatFromCurve(FloatCurve=DataInterface() {propertyValues {Curve {keys [<at 0.25 0.7>,<at 0.5 0.05>]}}})
    Index 0 Param 2=2
    Index 0 Param 3=1
    Index 0 Param 4=1
    Module.Param0WriteEnabled=true
    Module.Param1WriteEnabled=true
    Module.Param2WriteEnabled=true
    Module.Param3WriteEnabled=true
   )
   &CameraOffset(Camera Offset Amount=50)
  ]
  rendererClasses [&NiagaraSpriteRendererProperties]
  renderers [{rendererClass=&NiagaraSpriteRendererProperties}]
 }
 Hand_Arcs001 {
  emitterUpdateScript [
   &EmitterState(Inactive Response=::NewEnumerator1, Life Cycle Mode=::NewEnumerator1, Loop Behavior=::NewEnumerator1, Loop Duration=1, UseLoopDelay=false)
   &SpawnBurst_Instantaneous(Spawn Count=1, Use Spawn Probability=false) {enabled=false}
   &SpawnBurst_Instantaneous(Spawn Count=1, Spawn Group=1, Use Spawn Probability=false) {enabled=false, module=SpawnBurst_Instantaneous001}
   &SpawnRate(SpawnRate=4, Use Spawn Probability=false)
  ]
  particleSpawnScript [
   &FN_SetSpawnGroup_v2(Number of Spawn Groups=2)
   &InitializeParticle(
    Color=<rgba 0.2301 2.741 4 1>
    Lifetime=&UniformRangedFloat(Minimum=0.15, Maximum=0.35)
    Module.Write Color=true
    Module.Write Lifetime=true
    Module.Write Mass=true
    Module.Write Material Random=true
    Module.Write Position=false
    Module.Write Scale=false
    Module.Write SpriteRotation=true
    Module.Write SpriteSize=true
    Position=&SimulationPosition()
    Sprite Rotation=&UniformAOrBFloat(A=0, B=180)
    Sprite Size=&Multiply_Vector2(
     A=&UniformRangedVector2D(Minimum=<30 20>, Maximum=<20 15>)
     B=&Vector2DFromFloat(Value=&UniformAOrBFloat(A=-1))
    )
   )
   &CameraOffset(Camera Offset Amount=8)
  ]
  particleUpdateScript [
   &ParticleState()
   &SkeletalMeshLocation(
    Bone / Socket Index=&MakeIntFromFloat(Float=&Lerp_Float(A=4, B=5, Alpha=&MakeFloatFromInt(Int=&SpawnGroup())))
    Bone Sampling Mode=::NewEnumerator4
    Inherit Velocity=false
    MaskSpawnGroup=false
    Sampled Position Offset=&Lerp_Vector(A=<0 0 0>, B=<0 0 0>, Alpha=&MakeFloatFromInt(Int=&SpawnGroup()))
    Sampled Position Offset Coordinate Space:ENiagaraMeshTransforms=::NewEnumerator3
    Skeletal Mesh=DataInterface() {propertyValues {
      DefaultMesh=null
      ExcludeBoneName=root
      OptionalComponentClass=null
      PreviewMesh=&M_Med_Soldier_04_Preview
      SkinningMode=SkinOnTheFly
      SoftSourceActor=null
      SourceMode=Default
      UvSetIndex=0
      WholeMeshLOD=-1
      bExcludeBone=false
      bReadDeformedGeometry=true
      bRequireCurrentFrameData=true
      ComponentTags []
      FilteredBones [hand_r,lowerarm_r]
      FilteredSockets []
      MeshUserParameter {parameter {name=None, typeDefHandle {registeredTypeIndex=1}, varData []}}
      SamplingRegions []
     }}
   )
   &SkeletalMeshLocation(
    Apply To Particle Position=false
    Bone / Socket Index=&MakeIntFromFloat(Float=&Lerp_Float(A=8, B=9, Alpha=&MakeFloatFromInt(Int=$Particles.SpawnGroup)))
    Bone Sampling Mode=::NewEnumerator4
    Inherit Velocity=false
    Interpolate Mesh Sampling=false
    MaskSpawnGroup=false
    Skeletal Mesh=$User.SkeletalMesh
   ) {module=SkeletalMeshLocation001}
   /NiagaraFN/Sprite/FN_SpriteAlignmentAndFacing(SpriteAlignment:Vector3f=&Subtract_Vector(A=$Output.SkeletalMeshLocation001.SampledPosition, B=$Output.SkeletalMeshLocation.SampledPosition)) {module=SpriteAlignmentAndFacing}
   &DynamicMaterialParameters(
    Index 0 Param 1=&UniformRangedFloat(Minimum=2.5, Maximum=3.5)
    Index 0 Param 3=&UniformRangedFloat(Minimum=0.2, Maximum=0.5)
    Module.Param0WriteEnabled=true
    Module.Param1WriteEnabled=false
    Module.Param2WriteEnabled=true
    Module.Param3WriteEnabled=true
   )
   &ScaleColor(
    Scale Alpha=&FloatFromCurve(FloatCurve=DataInterface() {propertyValues {Curve {keys [<at 0 0 RCIM_Cubic>,<at 0.3 1>,<at 0.75 1 RCIM_Cubic>,<at 1 0 RCIM_Cubic>]}}})
    ScaleA=true
    ScaleRGB=true
   )
   &SpriteRotationRate(Rotation Rate=&UniformRangedFloat(Minimum=-0.1, Maximum=0.1))
   &ScaleSpriteSize(Scale Factor=/Niagara/DynamicInputs/ScaleByCurve/ScaleVector2DByCurve(
     InitialValue:Vector2f=<1 1>
     ScaleCurve:NiagaraDataInterfaceVectorCurve=DataInterface() {propertyValues {XCurve {keys [<at 0 1>,<at 0.0037 1>]}, YCurve {keys [<at 0 1>,<at 1 2>]}, ZCurve {keys [<at 0.0037 1>]}}}
     CurveIndex:NiagaraFloat=$Particles.NormalizedAge
    ))
   &AddVelocity(Velocity=&UniformRangedVector(Minimum=<0 0 15>, Maximum=<1 1 25>))
   &SolveForcesAndVelocity(Clamp Velocity=false, Limit Acceleration=false)
  ]
  rendererClasses [&NiagaraSpriteRendererProperties]
  renderers [{rendererClass=&NiagaraSpriteRendererProperties}]
 }
 Lightning_big02 {
  emitterUpdateScript [
   &EmitterState(Life Cycle Mode=::NewEnumerator1, Loop Behavior=::NewEnumerator1, Loop Duration=1, UseLoopDelay=false)
   &SpawnBurst_Instantaneous(Spawn Count=2, Use Spawn Probability=false)
   &SpawnBurst_Instantaneous(Spawn Count=1, Spawn Time=0.1, Use Spawn Probability=false) {module=SpawnBurst_Instantaneous001}
  ]
  particleSpawnScript [
   &InitializeParticle(
    Color=&MakeLinearColorFromVectorAndFloat("Vector (RGB)"=<1 1 1>, "Float (Alpha)"=5)
    Lifetime=&UniformRangedFloat(Minimum=0.1, Maximum=0.15)
    Mesh Scale=<0.17 0.17 0.18>
    Module.Write Color=true
    Module.Write Lifetime=true
    Module.Write Mass=true
    Module.Write Material Random=true
    Module.Write Position=false
    Module.Write Scale=true
    Module.Write SpriteRotation=false
    Module.Write SpriteSize=false
    Position=&SimulationPosition()
   )
   /Niagara/Modules/Spawn/Orientation/V2/InitialMeshOrientation(
    Orientation Vector=<0 0 -1>
    Rotation=&UniformRangedVector(Minimum=<0 0 0>, Maximum=<0 0 1>, Randomness Mode=::NewEnumerator1, Random Seed=10)
    Use Orientation Vector=true
    Use Rotation Vector=true
   )
   &AddVelocity(Coordinate Space=::World, Velocity=<0 0 1>)
  ]
  particleUpdateScript [
   &DynamicMaterialParameters(
    Index 0 Param 1=&UniformRangedFloat(Minimum=1, Maximum=5)
    Index 0 Param 2=&UniformRangedFloat(Minimum=0, Maximum=1)
    Index 0 Param 3=1
    Index 0 Param 4=1
    Module.Param0WriteEnabled=true
    Module.Param1WriteEnabled=true
    Module.Param2WriteEnabled=true
    Module.Param3WriteEnabled=true
   )
   &SkeletalMeshLocation(
    Inherit Velocity=false
    Interpolate Mesh Sampling=false
    MaskSpawnGroup=false
    Mesh Sampling Type=::NewEnumerator5
    Sampled Orientation Coordinate Space=::World
    Sampled Position Offset Coordinate Space:ENiagaraMeshTransforms=::NewEnumerator2
    Skeletal Mesh=$User.Weapon
    Socket Sampling Mode=::NewEnumerator5
    Use Sampled Mesh Attribute as Ribbon ID=false
   )
   &ParticleState()
   &SolveForcesAndVelocity(Clamp Velocity=false, Limit Acceleration=false)
  ]
  rendererClasses [&NiagaraMeshRendererProperties]
  renderers [{rendererClass=&NiagaraMeshRendererProperties, rendererIndex=0}]
 }
 Lightning_center {
  emitterUpdateScript [
   &EmitterState(Life Cycle Mode=::NewEnumerator1, Loop Behavior=::NewEnumerator1, Loop Duration=1, UseLoopDelay=false)
   &SpawnBurst_Instantaneous(Spawn Count=1, Use Spawn Probability=false)
   &SpawnBurst_Instantaneous(Spawn Count=1, Spawn Time=0.1, Use Spawn Probability=false) {module=SpawnBurst_Instantaneous001}
  ]
  particleSpawnScript [
   &InitializeParticle(
    Color=&MakeLinearColorFromVectorAndFloat(
     "Vector (RGB)"=<1 1 1>
     "Float (Alpha)"=&FloatFromCurve(FloatCurve=DataInterface() {propertyValues {Curve {keys [<at 0 5>,<at 1 0>]}}})
    )
    Lifetime=&UniformRangedFloat(Minimum=0.06, Maximum=0.08)
    Mesh Scale=&UniformRangedVector(Minimum=<1.2 0.4 0.4>, Maximum=<1.2 0.4 0.4>)
    Module.Write Color=true
    Module.Write Lifetime=true
    Module.Write Mass=true
    Module.Write Material Random=true
    Module.Write Position=false
    Module.Write Scale=true
    Module.Write SpriteRotation=false
    Module.Write SpriteSize=false
    Position=&SimulationPosition()
   )
   &InitialMeshOrientation(
    Rotation=&UniformRangedVector(Minimum=<0 0 0>, Maximum=<1 0 0>)
    Use Orientation Vector=true
    Use Rotation Vector=true
   )
   &DynamicMaterialParameters(
    Index 0 Param 1=20
    Index 0 Param 2=&UniformRangedFloat(Minimum=0.2, Maximum=0.2)
    Index 0 Param 3=1
    Index 0 Param 4=1
    Module.Param0WriteEnabled=true
    Module.Param1WriteEnabled=true
    Module.Param2WriteEnabled=true
    Module.Param3WriteEnabled=true
   )
   &AddVelocity(Coordinate Space=::World, Velocity=<0 0 1>)
  ]
  particleUpdateScript [
   &SkeletalMeshLocation(
    Inherit Velocity=false
    Interpolate Mesh Sampling=false
    MaskSpawnGroup=false
    Mesh Sampling Type=::NewEnumerator5
    Sampled Orientation Coordinate Space=::World
    Skeletal Mesh=$User.Weapon
    Socket Sampling Mode=::NewEnumerator5
   )
   &ParticleState()
   &SolveForcesAndVelocity(Clamp Velocity=false, Limit Acceleration=false)
   &ScaleColor(
    Scale Alpha=&FloatFromCurve(FloatCurve=DataInterface() {propertyValues {Curve {keys [<at 0 0>,<at 0.2 1>,<at 1 0>]}}})
    ScaleA=true
    ScaleRGB=true
   )
  ]
  rendererClasses [&NiagaraMeshRendererProperties]
  renderers [{rendererClass=&NiagaraMeshRendererProperties, rendererIndex=0}]
 }
 Lightning_residue {
  emitterUpdateScript [
   &EmitterState(Life Cycle Mode=::NewEnumerator1, Loop Behavior=::NewEnumerator1, Loop Duration=1, UseLoopDelay=false)
   &SpawnBurst_Instantaneous(Spawn Count=3, Spawn Time=0.18, Use Spawn Probability=false)
   &SpawnBurst_Instantaneous(Spawn Count=2, Spawn Time=0.3, Use Spawn Probability=false) {module=SpawnBurst_Instantaneous001}
  ]
  particleSpawnScript [
   &InitializeParticle(
    Color=<rgba 3.7748 6.4797 50 5>
    Lifetime=&UniformRangedFloat(Minimum=0.08, Maximum=0.1)
    Mesh Scale=&UniformRangedVector(Minimum=<3 3 8>, Maximum=<8 8 8>)
    Module.Write Color=true
    Module.Write Lifetime=true
    Module.Write Mass=true
    Module.Write Material Random=true
    Module.Write Position=false
    Module.Write Scale=true
    Module.Write SpriteRotation=false
    Module.Write SpriteSize=false
    Position=&SimulationPosition()
   )
   &InitialMeshOrientation(
    Rotation=&UniformRangedVector(Minimum=<0 0 0>, Maximum=<1 0 0>)
    Use Orientation Vector=true
    Use Rotation Vector=true
   ) {module=InitialMeshOrientation001}
   &DynamicMaterialParameters(
    Index 0 Param 1=&UniformRangedFloat(Minimum=5, Maximum=10)
    Index 0 Param 3=&UniformRangedFloat(Minimum=0.2, Maximum=1)
    Index 0 Param 4=&UniformRangedFloat(Minimum=0, Maximum=1)
    Module.Param0WriteEnabled=true
    Module.Param1WriteEnabled=false
    Module.Param2WriteEnabled=true
    Module.Param3WriteEnabled=true
   )
   &SphereLocation(MaskSpawnGroup=false, Module.Override Randomness=false, Sphere Radius=10, Surface Only=false)
   &AddVelocity(Coordinate Space=::World, Velocity=<0 0 1>)
  ]
  particleUpdateScript [
   &SkeletalMeshLocation(
    Inherit Velocity=false
    Interpolate Mesh Sampling=false
    MaskSpawnGroup=false
    Mesh Sampling Type=::NewEnumerator5
    Sampled Orientation Coordinate Space=::World
    Skeletal Mesh=$User.Weapon
    Socket Sampling Mode=::NewEnumerator5
   )
   &ParticleState()
   &SolveForcesAndVelocity(Clamp Velocity=false, Limit Acceleration=false)
   &ScaleColor(
    Scale Alpha=&FloatFromCurve(
     FloatCurve=DataInterface() {propertyValues {Curve {keys [<at 0 0>,<at 0.2 1>,<at 1 0>]}}}
     Scale Curve=&UniformRangedFloat(Minimum=0, Maximum=1)
    )
    Scale RGB=&MakeVectorFromLinearColorRGB(Color=&UniformRangedLinearColor(Minimum=<rgba 5.3114 10.1298 50 1>, Maximum=<rgba 0.1515 0.4587 2 1>))
    ScaleA=true
    ScaleRGB=true
   ) {enabled=false}
   &ScaleColor(
    Scale Alpha=&FloatFromCurve(
     FloatCurve=DataInterface() {propertyValues {Curve {keys [<at 0 1 RCIM_Constant>,<at 0.1189 0 RCIM_Constant>,<at 0.2516 0.997 RCIM_Constant>,<at 0.6735 -0.0069 RCIM_Constant>,<at 0.7649 1.0117 RCIM_Constant>,<at 1 0 RCIM_Constant>]}}}
     Scale Curve=&Multiply_Float(
      A=&FloatFromCurve(FloatCurve=DataInterface() {propertyValues {Curve {keys [<at 0 1 RCIM_Constant>,<at 0.1561 -0.0068 RCIM_Constant>,<at 0.3011 0.9877 RCIM_Constant>,<at 0.6424 0.0069 RCIM_Constant>,<at 0.7618 0.9877 RCIM_Constant>,<at 1 0 RCIM_Constant>]}}})
      B=&UniformRangedFloat(Minimum=1, Maximum=1)
     )
    )
    Scale RGB=&VectorFromFloat(Value=&UniformRangedFloat(Minimum=1, Maximum=5))
    ScaleA=true
    ScaleRGB=true
   ) {module=ScaleColor001}
   &DynamicMaterialParameters(
    Index 0 Param 2=&FloatFromCurve(FloatCurve=DataInterface() {propertyValues {Curve {keys [{interpMode=RCIM_Cubic, leaveTangent=-2.846, tangentMode=RCTM_Break, time=0, value=1}, <at 1 -0.25>]}}})
    Module.Param0WriteEnabled=false
    Module.Param1WriteEnabled=true
    Module.Param2WriteEnabled=false
    Module.Param3WriteEnabled=false
   ) {module=DynamicMaterialParameters001}
  ]
  rendererClasses [&NiagaraMeshRendererProperties]
  renderers [{rendererClass=&NiagaraMeshRendererProperties, rendererIndex=0}]
 }
 Lightning_small {
  emitterUpdateScript [
   &EmitterState(Inactive Response=::NewEnumerator1, Life Cycle Mode=::NewEnumerator1, Loop Duration=1, UseLoopDelay=false)
   &EmitterScalability(
    Module.bApplySpawnCountScale=false
    Module.bApplySpawnCountScaleByDistanceFraction=false
    Module.bUseMaxDistance=false
    Module.bUseMinDistance=false
   ) {enabled=false}
   &EmitterLifeCycle(CompleteOnInactive=true, MaxLoopCount=1, NextLoopDelay=0.9875, NextLoopDuration=3.55) {enabled=false}
   &SpawnRate(SpawnRate=35, Use Spawn Probability=false)
   &SpawnBurst_Instantaneous(Spawn Count=4, Use Spawn Probability=false) {enabled=false}
  ]
  particleSpawnScript [
   &InitializeParticle(
    Lifetime=&UniformRangedFloat(Minimum=0.1, Maximum=0.14)
    Mesh Scale=&Multiply_Vector(
     A=&UniformRangedVector(Minimum=<0.8 0.8 0.8>, Maximum=<0.5 0.5 0.5>)
     B=&VectorFromFloat(Value=1)
    )
    Module.Write Color=true
    Module.Write Lifetime=true
    Module.Write Mass=true
    Module.Write Material Random=true
    Module.Write Position=true
    Module.Write Scale=true
    Module.Write SpriteRotation=false
    Module.Write SpriteSize=false
    Position=&SimulationPosition()
    Sprite Size=<25 25>
   )
   &InitialMeshOrientation(
    Rotation=&UniformRangedVector(Minimum=<-1 0 0>, Maximum=<1 0 0>)
    Use Orientation Vector=true
    Use Rotation Vector=true
   )
   &SkeletalMeshLocation(
    Inherit Velocity=false
    Interpolate Mesh Sampling=false
    MaskSpawnGroup=false
    Mesh Sampling Type=::NewEnumerator6
    Module.Override Randomness=false
    Sampled Orientation Coordinate Space=::World
    Skeletal Mesh=DataInterface() {propertyValues {
      DefaultMesh=null
      ExcludeBoneName=root
      OptionalComponentClass=null
      PreviewMesh=&M_Med_Soldier_04_Preview
      SkinningMode=SkinOnTheFly
      SoftSourceActor=null
      SourceMode=Default
      UvSetIndex=0
      WholeMeshLOD=-1
      bExcludeBone=false
      bReadDeformedGeometry=true
      bRequireCurrentFrameData=true
      ComponentTags []
      FilteredBones [spine_05,spine_04,spine_02,upperarm_l,upperarm_r,lowerarm_l,lowerarm_r,hand_l,hand_r]
      FilteredSockets [FX_Melee_R_Speedy1]
      MeshUserParameter {parameter {name=None, typeDefHandle {registeredTypeIndex=1}, varData []}}
      SamplingRegions []
     }}
   )
   &SphereLocation(MaskSpawnGroup=false, Module.Override Randomness=false, Offset=<0 0 5>, Sphere Radius=5, Surface Only=true)
   &AddVelocityFromPoint(Module.Use Falloff Distance=false, Velocity Strength=0.01)
  ]
  particleUpdateScript [
   &SolveForcesAndVelocity(Clamp Velocity=false, Limit Acceleration=false)
   &ParticleState()
   &ScaleColor(
    Scale RGB=&MakeVectorFromLinearColorRGB(Color=&ColorFromCurve(ColorCurve=DataInterface() {propertyValues {
        AlphaCurve {keys [<at 0 1 RCIM_Cubic>,<at 1.0315 1 RCIM_Cubic>]}
        BlueCurve {keys [<at 0 500 RCIM_Cubic>]}
        GreenCurve {keys [<at 0 49.9996 RCIM_Cubic>]}
        RedCurve {keys [<at 0 25 RCIM_Cubic>]}
       }}))
    ScaleA=true
    ScaleRGB=true
   )
   &DynamicMaterialParameters(
    Index 0 Param 1=&UniformRangedFloat(Minimum=4, Maximum=7)
    Index 0 Param 2=&FloatFromCurve(FloatCurve=DataInterface() {propertyValues {Curve {keys [<at 0.1129 1>,<at 1 -1>]}}})
    Index 0 Param 3=&UniformRangedFloat(Minimum=0.9, Maximum=1.5)
    Module.Param0WriteEnabled=true
    Module.Param1WriteEnabled=true
    Module.Param2WriteEnabled=true
    Module.Param3WriteEnabled=true
   )
   /SetVariables(Particles.Light_Size=10) {isSetParameters=true, module=SetVariables_#1}
   /SetVariables(Particles.Light_Color:LinearColor=<rgba 0.5653 2.5237 5 1>) {isSetParameters=true, module=SetVariables_#2}
  ]
  rendererClasses [&NiagaraLightRendererProperties, &NiagaraMeshRendererProperties]
  renderers [
   {rendererClass=&NiagaraLightRendererProperties, rendererIndex=0}
   {rendererClass=&NiagaraMeshRendererProperties, rendererIndex=1}
  ]
 }
 Lightning_small001 {
  emitterUpdateScript [
   &EmitterState(Inactive Response=::NewEnumerator1, Life Cycle Mode=::NewEnumerator1, Loop Behavior=::NewEnumerator1, Loop Duration=1, UseLoopDelay=false)
   &EmitterScalability(
    Module.bApplySpawnCountScale=false
    Module.bApplySpawnCountScaleByDistanceFraction=false
    Module.bUseMaxDistance=false
    Module.bUseMinDistance=false
   ) {enabled=false}
   &EmitterLifeCycle(CompleteOnInactive=true, MaxLoopCount=1, NextLoopDelay=0.9875, NextLoopDuration=3.55) {enabled=false}
   &SpawnRate(SpawnRate=35, Use Spawn Probability=false) {enabled=false}
   &SpawnBurst_Instantaneous(Spawn Count=2, Use Spawn Probability=false)
   &SpawnBurst_Instantaneous(Spawn Count=2, Spawn Group=1, Spawn Time=0.2, Use Spawn Probability=false) {module=SpawnBurst_Instantaneous001}
  ]
  particleSpawnScript [
   &InitializeParticle(
    Lifetime=&UniformRangedFloat(Minimum=0.1, Maximum=0.14)
    Mesh Scale=&Multiply_Vector(
     A=&UniformRangedVector(Minimum=<1 2 2>, Maximum=<0.8 1.2 1>)
     B=&VectorFromFloat(Value=1.4)
    )
    Module.Write Color=true
    Module.Write Lifetime=true
    Module.Write Mass=true
    Module.Write Material Random=true
    Module.Write Position=true
    Module.Write Scale=true
    Module.Write SpriteRotation=false
    Module.Write SpriteSize=false
    Position=&SimulationPosition()
    Sprite Size=<25 25>
   )
   &InitialMeshOrientation(
    Rotation=&UniformRangedVector(Minimum=<-1 0 0>, Maximum=<1 0 0>)
    Use Orientation Vector=true
    Use Rotation Vector=true
   )
   &SphereLocation(MaskSpawnGroup=false, Module.Override Randomness=false, Offset=<0 0 5>, Sphere Radius=5, Surface Only=true)
   &AddVelocityFromPoint(Module.Use Falloff Distance=false, Velocity Strength=0.01)
  ]
  particleUpdateScript [
   &SkeletalMeshLocation(
    Inherit Velocity=false
    Interpolate Mesh Sampling=false
    MaskSpawnGroup=false
    Mesh Sampling Type=::NewEnumerator5
    Module.Override Randomness=false
    Sampled Orientation Coordinate Space=::World
    Skeletal Mesh=DataInterface() {propertyValues {
      DefaultMesh=null
      ExcludeBoneName=root
      OptionalComponentClass=null
      PreviewMesh=&M_Med_Soldier_04_Preview
      SkinningMode=SkinOnTheFly
      SoftSourceActor=null
      SourceMode=Default
      UvSetIndex=0
      WholeMeshLOD=-1
      bExcludeBone=false
      bReadDeformedGeometry=true
      bRequireCurrentFrameData=true
      ComponentTags []
      FilteredBones []
      FilteredSockets [FX_Melee_R_Speedy1, weapon_r_socket]
      MeshUserParameter {parameter {name=None, typeDefHandle {registeredTypeIndex=1}, varData []}}
      SamplingRegions []
     }}
   )
   &SolveForcesAndVelocity(Clamp Velocity=false, Limit Acceleration=false)
   &ParticleState()
   &ScaleColor(
    Scale RGB=&MakeVectorFromLinearColorRGB(Color=&ColorFromCurve(ColorCurve=DataInterface() {propertyValues {
        AlphaCurve {keys [<at 0 1 RCIM_Cubic>,<at 1.0315 1 RCIM_Cubic>]}
        BlueCurve {keys [<at 0 500 RCIM_Cubic>]}
        GreenCurve {keys [<at 0 49.9996 RCIM_Cubic>]}
        RedCurve {keys [<at 0 25 RCIM_Cubic>]}
       }}))
    ScaleA=true
    ScaleRGB=true
   )
   &DynamicMaterialParameters(
    Index 0 Param 1=&UniformRangedFloat(Minimum=4, Maximum=7)
    Index 0 Param 2=&FloatFromCurve(FloatCurve=DataInterface() {propertyValues {Curve {keys [<at 0.1129 1>,<at 1 -1>]}}})
    Index 0 Param 3=&UniformRangedFloat(Minimum=0.9, Maximum=1.5)
    Module.Param0WriteEnabled=true
    Module.Param1WriteEnabled=true
    Module.Param2WriteEnabled=true
    Module.Param3WriteEnabled=true
   )
   /SetVariables(Particles.Light_Size=10) {isSetParameters=true, module=SetVariables_#1}
   /SetVariables(Particles.Light_Color:LinearColor=<rgba 0.3622 0.5647 5 1>) {isSetParameters=true, module=SetVariables_#2}
  ]
  rendererClasses [&NiagaraLightRendererProperties, &NiagaraMeshRendererProperties]
  renderers [
   {rendererClass=&NiagaraLightRendererProperties, rendererIndex=0}
   {rendererClass=&NiagaraMeshRendererProperties, rendererIndex=1}
  ]
 }
 Lightning_small002 {
  emitterUpdateScript [
   &EmitterState(
    Inactive Response=::NewEnumerator1
    Life Cycle Mode=::NewEnumerator1
    Loop Behavior=::NewEnumerator1
    Loop Delay=0.2
    Loop Duration=1
    UseLoopDelay=true
   )
   &EmitterScalability(
    Module.bApplySpawnCountScale=false
    Module.bApplySpawnCountScaleByDistanceFraction=false
    Module.bUseMaxDistance=false
    Module.bUseMinDistance=false
   ) {enabled=false}
   &EmitterLifeCycle(CompleteOnInactive=true, MaxLoopCount=1, NextLoopDelay=0.9875, NextLoopDuration=3.55) {enabled=false}
   &SpawnRate(SpawnRate=35, Use Spawn Probability=false) {enabled=false}
   &SpawnBurst_Instantaneous(Spawn Count=2, Spawn Time=0.3, Use Spawn Probability=false)
   &SpawnBurst_Instantaneous(Spawn Count=2, Spawn Group=1, Spawn Time=0.4, Use Spawn Probability=false) {module=SpawnBurst_Instantaneous001}
  ]
  particleSpawnScript [
   &InitializeParticle(
    Lifetime=&UniformRangedFloat(Minimum=0.1, Maximum=0.14)
    Mesh Scale=&Multiply_Vector(
     A=&UniformRangedVector(Minimum=<1 2 2>, Maximum=<0.8 1 1>)
     B=&VectorFromFloat(Value=1.4)
    )
    Module.Write Color=true
    Module.Write Lifetime=true
    Module.Write Mass=true
    Module.Write Material Random=true
    Module.Write Position=true
    Module.Write Scale=true
    Module.Write SpriteRotation=false
    Module.Write SpriteSize=false
    Position=&SimulationPosition()
    Sprite Size=<25 25>
   )
   &InitialMeshOrientation(
    Rotation=&UniformRangedVector(Minimum=<0 0 0>, Maximum=<1 0 0>)
    Use Orientation Vector=true
    Use Rotation Vector=true
   )
   &SphereLocation(MaskSpawnGroup=false, Module.Override Randomness=false, Offset=<0 0 5>, Sphere Radius=5, Surface Only=true)
   &AddVelocityFromPoint(Module.Use Falloff Distance=false, Velocity Strength=0.01)
  ]
  particleUpdateScript [
   &SkeletalMeshLocation(
    Bone / Socket Index=&MakeIntFromFloat(Float=&Lerp_Float(A=0, B=1, Alpha=&MakeFloatFromInt(Int=$Emitter.SpawnGroup)))
    Bone Sampling Mode=::NewEnumerator4
    Inherit Velocity=false
    Interpolate Mesh Sampling=false
    MaskSpawnGroup=false
    Sampled Orientation Coordinate Space=::World
    Skeletal Mesh=DataInterface() {propertyValues {
      DefaultMesh=null
      ExcludeBoneName=root
      OptionalComponentClass=null
      PreviewMesh=&M_Med_Soldier_04_Preview
      SkinningMode=SkinOnTheFly
      SoftSourceActor=null
      SourceMode=Default
      UvSetIndex=0
      WholeMeshLOD=-1
      bExcludeBone=false
      bReadDeformedGeometry=true
      bRequireCurrentFrameData=true
      ComponentTags []
      FilteredBones [hand_r,lowerarm_r]
      FilteredSockets []
      MeshUserParameter {parameter {name=None, typeDefHandle {registeredTypeIndex=1}, varData []}}
      SamplingRegions []
     }}
    Spawn Group Mask=2
   )
   &SolveForcesAndVelocity(Clamp Velocity=false, Limit Acceleration=false)
   &ParticleState()
   &ScaleColor(
    Scale RGB=&MakeVectorFromLinearColorRGB(Color=&ColorFromCurve(ColorCurve=DataInterface() {propertyValues {
        AlphaCurve {keys [<at 0 1 RCIM_Cubic>,<at 1.0315 1 RCIM_Cubic>]}
        BlueCurve {keys [<at 0 500 RCIM_Cubic>]}
        GreenCurve {keys [<at 0 49.9996 RCIM_Cubic>]}
        RedCurve {keys [<at 0 25 RCIM_Cubic>]}
       }}))
    ScaleA=true
    ScaleRGB=true
   )
   &DynamicMaterialParameters(
    Index 0 Param 1=&UniformRangedFloat(Minimum=4, Maximum=7)
    Index 0 Param 2=&FloatFromCurve(FloatCurve=DataInterface() {propertyValues {Curve {keys [<at 0.1129 1>,<at 1 -1>]}}})
    Index 0 Param 3=&UniformRangedFloat(Minimum=0.9, Maximum=1.5)
    Module.Param0WriteEnabled=true
    Module.Param1WriteEnabled=true
    Module.Param2WriteEnabled=true
    Module.Param3WriteEnabled=true
   )
   /SetVariables(Particles.Light_Size=10) {isSetParameters=true, module=SetVariables_#1}
   /SetVariables(Particles.Light_Color:LinearColor=<rgba 0.3622 0.5647 5 1>) {isSetParameters=true, module=SetVariables_#2}
  ]
  rendererClasses [&NiagaraLightRendererProperties, &NiagaraMeshRendererProperties]
  renderers [
   {rendererClass=&NiagaraLightRendererProperties, rendererIndex=0}
   {rendererClass=&NiagaraMeshRendererProperties, rendererIndex=1}
  ]
 }
 Sparks {
  emitterUpdateScript [
   &EmitterState(Life Cycle Mode=::NewEnumerator1, Loop Behavior=::NewEnumerator1, Loop Delay=0.1, Loop Duration=0.3, UseLoopDelay=true)
   &SpawnRate(SpawnRate=5, Use Spawn Probability=false)
   &SpawnBurst_Instantaneous(Spawn Count=20, Use Spawn Probability=false)
  ]
  particleSpawnScript [
   &InitializeParticle(
    Color=<rgba 82.0913 575.8 1000 1>
    Lifetime=&UniformRangedFloat(Minimum=0.2, Maximum=0.5)
    Mesh Scale=<1.5 5 1>
    Module.Write Color=true
    Module.Write Lifetime=true
    Module.Write Mass=true
    Module.Write Material Random=true
    Module.Write Position=false
    Module.Write Scale=false
    Module.Write SpriteRotation=false
    Module.Write SpriteSize=true
    Position=&SimulationPosition()
    Sprite Size=&UniformRangedVector2D(Minimum=<1 5>, Maximum=<1 4>)
   )
   &AddVelocity(Coordinate Space=::World, Velocity=&UniformRangedVector(Minimum=<0 0 200>, Maximum=<1 1 500>))
   &SphereLocation(MaskSpawnGroup=false, Module.Override Randomness=false, Sphere Radius=5, Surface Only=true)
   &AddVelocityFromPoint(Module.Use Falloff Distance=false, Velocity Strength=200)
   &SkeletalMeshLocation(
    Inherit Velocity=false
    Interpolate Mesh Sampling=false
    MaskSpawnGroup=false
    Mesh Sampling Type=::NewEnumerator5
    Module.Override Randomness=false
    Sampled Orientation Coordinate Space=::World
    Skeletal Mesh=DataInterface() {propertyValues {
      DefaultMesh=null
      ExcludeBoneName=root
      OptionalComponentClass=null
      PreviewMesh=&M_Med_Soldier_04_Preview
      SkinningMode=SkinOnTheFly
      SoftSourceActor=null
      SourceMode=Default
      UvSetIndex=0
      WholeMeshLOD=-1
      bExcludeBone=false
      bReadDeformedGeometry=true
      bRequireCurrentFrameData=true
      ComponentTags []
      FilteredBones []
      FilteredSockets [FX_Melee_R_Speedy1]
      MeshUserParameter {parameter {name=None, typeDefHandle {registeredTypeIndex=1}, varData []}}
      SamplingRegions []
     }}
   )
  ]
  particleUpdateScript [
   &GravityForce(Gravity=<0 0 -1500>)
   &ParticleState()
   &InheritVelocity(Module.Limit Speed=true, Velocity Limit=20)
   &SolveForcesAndVelocity(Clamp Velocity=false, Limit Acceleration=false)
   &ScaleColor(
    Scale Alpha=&FloatFromCurve(FloatCurve=DataInterface() {propertyValues {Curve {keys [
         {arriveTangent=-2.7664, interpMode=RCIM_Cubic, leaveTangent=-2.7664, tangentMode=RCTM_User, time=0, value=0.5}
         {arriveTangent=2, interpMode=RCIM_Cubic, leaveTangent=2, tangentMode=RCTM_Break, time=0.2, value=0.2621}
         {arriveTangent=-0.3277, interpMode=RCIM_Cubic, leaveTangent=-3.899, tangentMode=RCTM_Break, time=0.25, value=1}
         <at 1 0>
        ]}}})
    ScaleA=true
    ScaleRGB=true
   )
   &DynamicMaterialParameters(
    Index 0 Param 1=&FloatFromCurve(FloatCurve=DataInterface() {propertyValues {Curve {keys [<at 0.25 0.7>,<at 0.5 0.05>]}}})
    Index 0 Param 2=2
    Index 0 Param 3=1
    Index 0 Param 4=1
    Module.Param0WriteEnabled=true
    Module.Param1WriteEnabled=true
    Module.Param2WriteEnabled=true
    Module.Param3WriteEnabled=true
   )
   &CameraOffset(Camera Offset Amount=50)
   &ScaleSpriteSize(Scale Factor=&Vector2DFromCurve(
     Vector2Curve=DataInterface() {propertyValues {XCurve {keys [<at 0 1.5>,<at 0.3 1.262>,<at 1 1>]}, YCurve {keys [<at 0 1>,<at 0.3 1.1>,<at 1 0>]}}}
     ScaleCurve=<1 1>
    ))
  ]
  rendererClasses [&NiagaraSpriteRendererProperties]
  renderers [{rendererClass=&NiagaraSpriteRendererProperties}]
 }
 Start_light {
  emitterUpdateScript [
   &EmitterState(
    Inactive Response=::NewEnumerator1
    Life Cycle Mode=::NewEnumerator1
    Loop Behavior=::NewEnumerator1
    Loop Delay=&UniformRangedFloat(Minimum=0.5, Maximum=1)
    Loop Duration=1
    UseLoopDelay=false
   )
   &SpawnRate(SpawnRate=0, Use Spawn Probability=false)
   &SpawnBurst_Instantaneous(Spawn Count=1, Use Spawn Probability=false)
  ]
  particleSpawnScript [
   &InitializeParticle(
    Color=<rgba 0.0247 0.0961 5 1>
    Lifetime=0.2
    Mesh Scale=&Multiply_Vector(
     A=<3 2.5 2.5>
     B=&VectorFromFloat(Value=&UniformRangedFloat(Minimum=0.5, Maximum=1))
    )
    Module.Write Color=true
    Module.Write Lifetime=true
    Module.Write Mass=true
    Module.Write Material Random=true
    Module.Write Position=false
    Module.Write Scale=false
    Module.Write SpriteRotation=true
    Module.Write SpriteSize=true
    Position=&SimulationPosition()
    Sprite Rotation=&UniformRangedFloat(Minimum=0, Maximum=1)
    Sprite Size=<200 200>
   )
   &DynamicMaterialParameters(
    Index 0 Param 2=2
    Index 0 Param 3=1
    Index 0 Param 4=1
    Module.Param0WriteEnabled=false
    Module.Param1WriteEnabled=true
    Module.Param2WriteEnabled=true
    Module.Param3WriteEnabled=true
   )
  ]
  particleUpdateScript [
   &SkeletalMeshLocation(
    Inherit Velocity=false
    Interpolate Mesh Sampling=false
    MaskSpawnGroup=false
    Mesh Sampling Type=::NewEnumerator5
    Skeletal Mesh=$User.Weapon
    Socket Sampling Mode=::NewEnumerator5
   )
   &ParticleState()
   &SolveForcesAndVelocity(Clamp Velocity=false, Limit Acceleration=false)
   &DynamicMaterialParameters(
    Index 0 Param 1=&FloatFromCurve(FloatCurve=DataInterface() {propertyValues {Curve {keys [<at 0.25 0.7>,<at 0.5 0.05>,<at 1 0>]}}})
    Module.Param0WriteEnabled=true
    Module.Param1WriteEnabled=false
    Module.Param2WriteEnabled=false
    Module.Param3WriteEnabled=false
   ) {module=DynamicMaterialParameters002}
   &ScaleColor(
    Scale Alpha=&FloatFromCurve(FloatCurve=DataInterface() {propertyValues {Curve {keys [<at 0 0.5>,<at 0.2 0.262>,<at 0.25 1>,<at 1 0>]}}})
    ScaleA=true
    ScaleRGB=true
   )
   /SetVariables(Particles.LightRadius=&UniformRangedFloat(Minimum=50, Maximum=100)) {isSetParameters=true, module=SetVariables_#3}
   /SetVariables(Particles.light color:LinearColor=&MakeLinearColorFromVectorAndFloat(
     "Vector (RGB)"=&MakeVectorFromLinearColorRGB(Color=<rgba 1.1934 4.0631 20 1>)
     "Float (Alpha)"=&FloatFromCurve(FloatCurve=DataInterface() {propertyValues {Curve {keys [<at 0 1>,<at 0.0918 0.159>,<at 0.1514 0.9748>,<at 0.2547 0.1132>,<at 0.3492 0.8648>,<at 0.3825 0.1223>,<at 0.4928 0.8923>,<at 0.5611 0.1407>,<at 0.6276 0.9107>,<at 0.6959 0.1407>,<at 0.7677 0.874>,<at 0.85 0.1315>,<at 0.9235 0.874>,<at 1 0>]}}})
    )) {isSetParameters=true, module=SetVariables_#4}
   &CameraOffset()
  ]
  rendererClasses [&NiagaraLightRendererProperties]
  renderers [{rendererClass=&NiagaraLightRendererProperties, rendererIndex=0}]
 }
 glow {
  emitterUpdateScript [
   &EmitterState(
    Inactive Response=::NewEnumerator1
    Life Cycle Mode=::NewEnumerator1
    Loop Behavior=::NewEnumerator1
    Loop Delay=0.3
    Loop Duration=0.2
    UseLoopDelay=true
   )
   &SpawnRate(SpawnRate=5, Use Spawn Probability=false)
   &SpawnBurst_Instantaneous(Spawn Count=1, Use Spawn Probability=false)
  ]
  particleSpawnScript [
   &InitializeParticle(
    Color=<rgba 0.0247 0.0961 5 0.4>
    Lifetime=0.8
    Mesh Scale=&Multiply_Vector(
     A=<3 2.5 2.5>
     B=&VectorFromFloat(Value=&UniformRangedFloat(Minimum=0.5, Maximum=1))
    )
    Module.Write Color=true
    Module.Write Lifetime=true
    Module.Write Mass=true
    Module.Write Material Random=true
    Module.Write Position=false
    Module.Write Scale=false
    Module.Write SpriteRotation=true
    Module.Write SpriteSize=true
    Position=&SimulationPosition()
    Sprite Rotation=&UniformRangedFloat(Minimum=0, Maximum=1)
    Sprite Size=<320 320>
   )
   &DynamicMaterialParameters(
    Index 0 Param 2=2
    Index 0 Param 3=1
    Index 0 Param 4=1
    Module.Param0WriteEnabled=false
    Module.Param1WriteEnabled=true
    Module.Param2WriteEnabled=true
    Module.Param3WriteEnabled=true
   )
  ]
  particleUpdateScript [
   &SkeletalMeshLocation(
    Inherit Velocity=false
    Interpolate Mesh Sampling=false
    MaskSpawnGroup=false
    Mesh Sampling Type=::NewEnumerator5
    Sampled Position Offset=<0 -5 0>
    Skeletal Mesh=$User.Weapon
    Socket Sampling Mode=::NewEnumerator5
   )
   &ParticleState()
   &SolveForcesAndVelocity(Clamp Velocity=false, Limit Acceleration=false)
   &DynamicMaterialParameters(
    Index 0 Param 1=&FloatFromCurve(FloatCurve=DataInterface() {propertyValues {Curve {keys [<at 0.25 0.7>,<at 0.5 0.05>,<at 1 0>]}}})
    Module.Param0WriteEnabled=true
    Module.Param1WriteEnabled=false
    Module.Param2WriteEnabled=false
    Module.Param3WriteEnabled=false
   ) {module=DynamicMaterialParameters002}
   &ScaleColor(
    Scale Alpha=&FloatFromCurve(FloatCurve=DataInterface() {propertyValues {Curve {keys [<at 0 0.5>,<at 0.2 0.262>,<at 0.25 1>,<at 1 0>]}}}, Scale Curve=0.1)
    ScaleA=true
    ScaleRGB=true
   )
   /SetVariables(Particles.LightRadius=&UniformRangedFloat(Minimum=20, Maximum=100)) {isSetParameters=true, module=SetVariables_#3}
   /SetVariables(Particles.light color:LinearColor=&MakeLinearColorFromVectorAndFloat(
     "Vector (RGB)"=&MakeVectorFromLinearColorRGB(Color=<rgba 1.1934 4.0631 20 1>)
     "Float (Alpha)"=&FloatFromCurve(FloatCurve=DataInterface() {propertyValues {Curve {keys [<at 0 1>,<at 0.0918 0.159>,<at 0.1514 0.9748>,<at 0.2547 0.1132>,<at 0.3492 0.8648>,<at 0.3825 0.1223>,<at 0.4928 0.8923>,<at 0.5611 0.1407>,<at 0.6276 0.9107>,<at 0.6959 0.1407>,<at 0.7677 0.874>,<at 0.85 0.1315>,<at 0.9235 0.874>,<at 1 0>]}}})
    )) {isSetParameters=true, module=SetVariables_#4}
   &CameraOffset(Camera Offset Amount=50)
  ]
  rendererClasses [&NiagaraSpriteRendererProperties]
  renderers [{rendererClass=&NiagaraSpriteRendererProperties}]
 }
 light {
  emitterUpdateScript [
   &EmitterState(
    Inactive Response=::NewEnumerator1
    Life Cycle Mode=::NewEnumerator1
    Loop Delay=&UniformRangedFloat(Minimum=0.5, Maximum=1)
    Loop Duration=0.2
    UseLoopDelay=false
   )
   &SpawnRate(SpawnRate=5, Use Spawn Probability=false)
   &SpawnBurst_Instantaneous(Spawn Count=1, Use Spawn Probability=false)
  ]
  particleSpawnScript [
   &InitializeParticle(
    Color=<rgba 0.0247 0.0961 5 1>
    Lifetime=0.4
    Mesh Scale=&Multiply_Vector(
     A=<3 2.5 2.5>
     B=&VectorFromFloat(Value=&UniformRangedFloat(Minimum=0.5, Maximum=1))
    )
    Module.Write Color=true
    Module.Write Lifetime=true
    Module.Write Mass=true
    Module.Write Material Random=true
    Module.Write Position=false
    Module.Write Scale=false
    Module.Write SpriteRotation=true
    Module.Write SpriteSize=true
    Position=&SimulationPosition()
    Sprite Rotation=&UniformRangedFloat(Minimum=0, Maximum=1)
    Sprite Size=<200 200>
   )
   &DynamicMaterialParameters(
    Index 0 Param 2=2
    Index 0 Param 3=1
    Index 0 Param 4=1
    Module.Param0WriteEnabled=false
    Module.Param1WriteEnabled=true
    Module.Param2WriteEnabled=true
    Module.Param3WriteEnabled=true
   )
  ]
  particleUpdateScript [
   &SkeletalMeshLocation(
    Inherit Velocity=false
    Interpolate Mesh Sampling=false
    MaskSpawnGroup=false
    Mesh Sampling Type=::NewEnumerator5
    Skeletal Mesh=$User.Weapon
    Socket Sampling Mode=::NewEnumerator5
   )
   &ParticleState()
   &SolveForcesAndVelocity(Clamp Velocity=false, Limit Acceleration=false)
   &DynamicMaterialParameters(
    Index 0 Param 1=&FloatFromCurve(FloatCurve=DataInterface() {propertyValues {Curve {keys [<at 0.25 0.7>,<at 0.5 0.05>,<at 1 0>]}}})
    Module.Param0WriteEnabled=true
    Module.Param1WriteEnabled=false
    Module.Param2WriteEnabled=false
    Module.Param3WriteEnabled=false
   ) {module=DynamicMaterialParameters002}
   &ScaleColor(
    Scale Alpha=&FloatFromCurve(FloatCurve=DataInterface() {propertyValues {Curve {keys [<at 0 0.5>,<at 0.2 0.262>,<at 0.25 1>,<at 1 0>]}}})
    ScaleA=true
    ScaleRGB=true
   )
   /SetVariables(Particles.LightRadius=&UniformRangedFloat(Minimum=50, Maximum=100)) {isSetParameters=true, module=SetVariables_#3}
   /SetVariables(Particles.light color:LinearColor=&MakeLinearColorFromVectorAndFloat(
     "Vector (RGB)"=&MakeVectorFromLinearColorRGB(Color=<rgba 0.5967 2.0315 10 1>)
     "Float (Alpha)"=&FloatFromCurve(FloatCurve=DataInterface() {propertyValues {Curve {keys [<at 0 1>,<at 0.0918 0.159>,<at 0.1514 0.9748>,<at 0.2547 0.1132>,<at 0.3492 0.8648>,<at 0.3825 0.1223>,<at 0.4928 0.8923>,<at 0.5611 0.1407>,<at 0.6276 0.9107>,<at 0.6959 0.1407>,<at 0.7677 0.874>,<at 0.85 0.1315>,<at 0.9235 0.874>,<at 1 0>]}}})
    )) {isSetParameters=true, module=SetVariables_#4}
   &CameraOffset()
  ]
  rendererClasses [&NiagaraLightRendererProperties]
  renderers [{rendererClass=&NiagaraLightRendererProperties, rendererIndex=0}]
 }
 light_of_big_lightning {
  emitterUpdateScript [
   &EmitterState(
    Inactive Response=::NewEnumerator1
    Life Cycle Mode=::NewEnumerator1
    Loop Behavior=::NewEnumerator1
    Loop Delay=&UniformRangedFloat(Minimum=0.5, Maximum=1)
    Loop Duration=1
    UseLoopDelay=false
   )
   &SpawnRate(SpawnRate=0, Use Spawn Probability=false)
   &SpawnBurst_Instantaneous(Spawn Count=1, Use Spawn Probability=false)
  ]
  particleSpawnScript [
   &InitializeParticle(
    Color=<rgba 0.0247 0.0961 5 1>
    Lifetime=0.18
    Mesh Scale=&Multiply_Vector(
     A=<3 2.5 2.5>
     B=&VectorFromFloat(Value=&UniformRangedFloat(Minimum=0.5, Maximum=1))
    )
    Module.Write Color=true
    Module.Write Lifetime=true
    Module.Write Mass=true
    Module.Write Material Random=true
    Module.Write Position=false
    Module.Write Scale=false
    Module.Write SpriteRotation=true
    Module.Write SpriteSize=true
    Position=&SimulationPosition()
    Sprite Rotation=&UniformRangedFloat(Minimum=0, Maximum=1)
    Sprite Size=<200 200>
   )
   &DynamicMaterialParameters(
    Index 0 Param 2=2
    Index 0 Param 3=1
    Index 0 Param 4=1
    Module.Param0WriteEnabled=false
    Module.Param1WriteEnabled=true
    Module.Param2WriteEnabled=true
    Module.Param3WriteEnabled=true
   )
  ]
  particleUpdateScript [
   &SkeletalMeshLocation(
    Inherit Velocity=false
    Interpolate Mesh Sampling=false
    MaskSpawnGroup=false
    Mesh Sampling Type=::NewEnumerator5
    Skeletal Mesh=$User.Weapon
    Socket Sampling Mode=::NewEnumerator5
   )
   &ParticleState()
   &SolveForcesAndVelocity(Clamp Velocity=false, Limit Acceleration=false)
   &DynamicMaterialParameters(
    Index 0 Param 1=&FloatFromCurve(FloatCurve=DataInterface() {propertyValues {Curve {keys [<at 0.25 0.7>,<at 0.5 0.05>,<at 1 0>]}}})
    Module.Param0WriteEnabled=true
    Module.Param1WriteEnabled=false
    Module.Param2WriteEnabled=false
    Module.Param3WriteEnabled=false
   ) {module=DynamicMaterialParameters002}
   &ScaleColor(
    Scale Alpha=&FloatFromCurve(FloatCurve=DataInterface() {propertyValues {Curve {keys [<at 0 0.5>,<at 0.2 0.262>,<at 0.25 1>,<at 1 0>]}}})
    ScaleA=true
    ScaleRGB=true
   )
   /SetVariables(Particles.LightRadius=&UniformRangedFloat(Minimum=50, Maximum=100)) {isSetParameters=true, module=SetVariables_#3}
   /SetVariables(Particles.light color:LinearColor=&MakeLinearColorFromVectorAndFloat(
     "Vector (RGB)"=&MakeVectorFromLinearColorRGB(Color=<rgba 7.757 26.41 130 1>)
     "Float (Alpha)"=&FloatFromCurve(FloatCurve=DataInterface() {propertyValues {Curve {keys [<at 0 1>,<at 0.0918 0.159>,<at 0.1514 0.9748>,<at 0.2547 0.1132>,<at 0.3492 0.8648>,<at 0.3825 0.1223>,<at 0.4928 0.8923>,<at 0.5611 0.1407>,<at 0.6276 0.9107>,<at 0.6959 0.1407>,<at 0.7677 0.874>,<at 0.85 0.1315>,<at 0.9235 0.874>,<at 1 0>]}}})
    )) {isSetParameters=true, module=SetVariables_#4}
   &CameraOffset()
  ]
  rendererClasses [&NiagaraLightRendererProperties]
  renderers [{rendererClass=&NiagaraLightRendererProperties, rendererIndex=0}]
 }
 lightning_burst {
  emitterUpdateScript [
   &EmitterState(
    Inactive Response=::NewEnumerator1
    Life Cycle Mode=::NewEnumerator1
    Loop Behavior=::NewEnumerator1
    Loop Delay=0.2
    Loop Duration=&UniformRangedFloat(Minimum=0.5, Maximum=1)
    UseLoopDelay=true
   )
   &SpawnRate(SpawnRate=3, Use Spawn Probability=false) {enabled=false}
   &SpawnBurst_Instantaneous(Spawn Count=4, Use Spawn Probability=false)
  ]
  particleSpawnScript [
   &InitializeParticle(
    Lifetime=&UniformRangedFloat(Minimum=0.1, Maximum=0.2)
    Mesh Scale=&UniformRangedVector(Minimum=<0.1 0.1 0.1>, Maximum=<0.4 0.4 0.4>)
    Module.Write Color=true
    Module.Write Lifetime=true
    Module.Write Mass=true
    Module.Write Material Random=true
    Module.Write Position=false
    Module.Write Scale=true
    Module.Write SpriteRotation=false
    Module.Write SpriteSize=false
    Position=&SimulationPosition()
    Sprite Size=&UniformRangedVector2D(Minimum=<0 0>, Maximum=<1 1>)
   )
   &AddVelocity(Velocity=&UniformRangedVector(Minimum=<-1 -1 -1>, Maximum=<1 1 1>)) {enabled=false}
   &InitialMeshOrientation(
    Orientation Vector=<0 1 0>
    Rotation=&UniformRangedVector(Minimum=<0 -1 0>, Maximum=<0 1 0>)
    Use Orientation Vector=true
    Use Rotation Vector=true
   )
   &DynamicMaterialParameters(
    Index 0 Param 2=&UniformRangedFloat(Minimum=0.3, Maximum=0.4)
    Module.Param0WriteEnabled=false
    Module.Param1WriteEnabled=true
    Module.Param2WriteEnabled=false
    Module.Param3WriteEnabled=false
   ) {module=DynamicMaterialParameters001}
  ]
  particleUpdateScript [
   &SkeletalMeshLocation(
    Inherit Velocity=false
    Interpolate Mesh Sampling=false
    MaskSpawnGroup=false
    Mesh Sampling Type=::NewEnumerator5
    Sampled Orientation Coordinate Space=::World
    Skeletal Mesh=$User.Weapon
    Socket Sampling Mode=::NewEnumerator5
   )
   &ParticleState()
   &SolveForcesAndVelocity(Clamp Velocity=false, Limit Acceleration=false)
   &ScaleColor(
    Scale Alpha=&FloatFromCurve(FloatCurve=DataInterface() {propertyValues {Curve {keys [<at 0 1>,<at 1 0>]}}})
    ScaleA=true
    ScaleRGB=true
   ) {enabled=false}
   &ScaleColor(
    Scale Alpha=&FloatFromCurve(FloatCurve=DataInterface() {propertyValues {Curve {keys [<at 0.15 1 RCIM_Cubic>, {arriveTangent=-1.1765, interpMode=RCIM_Cubic, leaveTangent=-1.1765, time=0.796, value=0.245}, <at 1 0>]}}})
    Scale RGB=&VectorFromFloat(Value=&FloatFromCurve(FloatCurve=DataInterface() {propertyValues {Curve {keys [<at 0 1>,<at 0.1242 0.9958 RCIM_Constant>,<at 0.1835 0.0197 RCIM_Constant>,<at 0.2946 0.9836 RCIM_Constant>,<at 0.502 0.0075 RCIM_Constant>,<at 0.6095 1.0202 RCIM_Constant>,<at 0.8012 0 RCIM_Constant>,<at 0.8836 1.008 RCIM_Constant>,<at 1 0>]}}}))
    ScaleA=true
    ScaleRGB=true
   ) {enabled=false, module=ScaleColor001}
   &ScaleColor(
    Scale Alpha=&FloatFromCurve(FloatCurve=DataInterface() {propertyValues {Curve {keys [<at 0.15 1 RCIM_Cubic>, {arriveTangent=-1.1765, interpMode=RCIM_Cubic, leaveTangent=-1.1765, time=0.796, value=0.245}, <at 1 0>]}}})
    Scale RGB=&VectorFromFloat(Value=&FloatFromCurve(FloatCurve=DataInterface() {propertyValues {Curve {keys [<at 0 1>,<at 0.1242 0.9958 RCIM_Constant>,<at 0.1835 0.0197 RCIM_Constant>,<at 0.2645 0.9931 RCIM_Constant>,<at 0.5317 -0.0047 RCIM_Constant>,<at 0.6798 0.9836 RCIM_Constant>,<at 1 0>]}}}))
    ScaleA=true
    ScaleRGB=true
   ) {enabled=false, module=ScaleColor002}
   &ScaleColor(
    Scale Alpha=&FloatFromCurve(FloatCurve=DataInterface() {propertyValues {Curve {keys [<at 0.15 1 RCIM_Cubic>, {arriveTangent=-1.1765, interpMode=RCIM_Cubic, leaveTangent=-1.1765, time=0.796, value=0.245}, <at 1 0>]}}})
    Scale RGB=&VectorFromFloat(Value=&FloatFromCurve(FloatCurve=DataInterface() {propertyValues {Curve {keys [<at 0 1>,<at 0.1242 0.9958 RCIM_Constant>,<at 0.1835 0 RCIM_Constant>,<at 0.284 0.984 RCIM_Constant>,<at 0.5317 0 RCIM_Constant>,<at 0.6402 0.984 RCIM_Constant>,<at 1 0>]}}}))
    ScaleA=true
    ScaleRGB=true
   ) {module=ScaleColor003}
   &DynamicMaterialParameters(
    Index 0 Param 1=&FloatFromCurve(FloatCurve=DataInterface() {propertyValues {Curve {keys [<at 0 1>,<at 0.7 0.35>,<at 1 0>]}}})
    Index 0 Param 2=0.3
    Index 0 Param 3=&FloatFromCurve(FloatCurve=DataInterface() {propertyValues {Curve {keys [<at 0.6 1>,<at 1 0>]}}}, Scale Curve=0.5)
    Index 0 Param 4=1
    Module.Param0WriteEnabled=true
    Module.Param1WriteEnabled=false
    Module.Param2WriteEnabled=true
    Module.Param3WriteEnabled=true
   )
  ]
  rendererClasses [&NiagaraMeshRendererProperties]
  renderers [{rendererClass=&NiagaraMeshRendererProperties, rendererIndex=0}]
 }
 lightning_burst001 {
  emitterUpdateScript [
   &EmitterState(
    Inactive Response=::NewEnumerator1
    Life Cycle Mode=::NewEnumerator1
    Loop Behavior=::NewEnumerator1
    Loop Delay=0.2
    Loop Duration=&UniformRangedFloat(Minimum=0.5, Maximum=1)
    UseLoopDelay=true
   )
   &SpawnRate(SpawnRate=3, Use Spawn Probability=false) {enabled=false}
   &SpawnBurst_Instantaneous(Spawn Count=1, Use Spawn Probability=false)
  ]
  particleSpawnScript [
   &InitializeParticle(
    Lifetime=&UniformRangedFloat(Minimum=0.12, Maximum=0.2)
    Mesh Scale=&UniformRangedVector(Minimum=<0.1 0.1 0.1>, Maximum=<0.4 0.4 0.4>)
    Module.Write Color=true
    Module.Write Lifetime=true
    Module.Write Mass=true
    Module.Write Material Random=true
    Module.Write Position=false
    Module.Write Scale=true
    Module.Write SpriteRotation=false
    Module.Write SpriteSize=false
    Position=&SimulationPosition()
    Sprite Size=&UniformRangedVector2D(Minimum=<0 0>, Maximum=<1 1>)
   )
   &AddVelocity(Velocity=&UniformRangedVector(Minimum=<-1 -1 -1>, Maximum=<1 1 1>)) {enabled=false}
   &InitialMeshOrientation(
    Orientation Vector=&UniformRangedVector(Minimum=<0 0 0>, Maximum=<0 0 0>)
    Rotation=&UniformRangedVector(Minimum=<0 -0.25 0>, Maximum=<0 -0.25 0>)
    Rotation Coordinate Space:ENiagaraMeshTransforms=::NewEnumerator1
    Use Orientation Vector=true
    Use Rotation Vector=true
   )
   &DynamicMaterialParameters(
    Index 0 Param 2=&UniformRangedFloat(Minimum=0.3, Maximum=0.4)
    Module.Param0WriteEnabled=false
    Module.Param1WriteEnabled=true
    Module.Param2WriteEnabled=false
    Module.Param3WriteEnabled=false
   ) {module=DynamicMaterialParameters001}
  ]
  particleUpdateScript [
   &SkeletalMeshLocation(
    Bone Sampling Mode=::NewEnumerator0
    Inherit Velocity=false
    Interpolate Mesh Sampling=false
    MaskSpawnGroup=false
    Module.Override Randomness=false
    Sampled Orientation Coordinate Space=::World
    Skeletal Mesh=DataInterface() {propertyValues {
      DefaultMesh=null
      ExcludeBoneName=root
      OptionalComponentClass=null
      PreviewMesh=&M_Med_Soldier_04_Preview
      SkinningMode=SkinOnTheFly
      SoftSourceActor=null
      SourceMode=Default
      UvSetIndex=0
      WholeMeshLOD=-1
      bExcludeBone=false
      bReadDeformedGeometry=true
      bRequireCurrentFrameData=true
      ComponentTags []
      FilteredBones [hand_r]
      FilteredSockets []
      MeshUserParameter {parameter {name=None, typeDefHandle {registeredTypeIndex=1}, varData []}}
      SamplingRegions []
     }}
   )
   &ParticleState()
   &SolveForcesAndVelocity(Clamp Velocity=false, Limit Acceleration=false)
   &ScaleColor(
    Scale Alpha=&FloatFromCurve(FloatCurve=DataInterface() {propertyValues {Curve {keys [<at 0 1>,<at 1 0>]}}})
    ScaleA=true
    ScaleRGB=true
   ) {enabled=false}
   &ScaleColor(
    Scale Alpha=&FloatFromCurve(FloatCurve=DataInterface() {propertyValues {Curve {keys [<at 0.15 1 RCIM_Cubic>, {arriveTangent=-1.1765, interpMode=RCIM_Cubic, leaveTangent=-1.1765, time=0.796, value=0.245}, <at 1 0>]}}})
    Scale RGB=&VectorFromFloat(Value=&FloatFromCurve(FloatCurve=DataInterface() {propertyValues {Curve {keys [<at 0 1>,<at 0.1242 0.9958 RCIM_Constant>,<at 0.1835 0.0197 RCIM_Constant>,<at 0.2946 0.9836 RCIM_Constant>,<at 0.502 0.0075 RCIM_Constant>,<at 0.6095 1.0202 RCIM_Constant>,<at 0.8012 0 RCIM_Constant>,<at 0.8836 1.008 RCIM_Constant>,<at 1 0>]}}}))
    ScaleA=true
    ScaleRGB=true
   ) {enabled=false, module=ScaleColor001}
   &ScaleColor(
    Scale Alpha=&FloatFromCurve(FloatCurve=DataInterface() {propertyValues {Curve {keys [<at 0.15 1 RCIM_Cubic>, {arriveTangent=-1.1765, interpMode=RCIM_Cubic, leaveTangent=-1.1765, time=0.796, value=0.245}, <at 1 0>]}}})
    Scale RGB=&VectorFromFloat(Value=&FloatFromCurve(FloatCurve=DataInterface() {propertyValues {Curve {keys [<at 0 1>,<at 0.1242 0.9958 RCIM_Constant>,<at 0.1835 0.0197 RCIM_Constant>,<at 0.2645 0.9931 RCIM_Constant>,<at 0.5317 -0.0047 RCIM_Constant>,<at 0.6798 0.9836 RCIM_Constant>,<at 1 0>]}}}))
    ScaleA=true
    ScaleRGB=true
   ) {enabled=false, module=ScaleColor002}
   &ScaleColor(
    Scale Alpha=&FloatFromCurve(FloatCurve=DataInterface() {propertyValues {Curve {keys [<at 0.15 1 RCIM_Cubic>, {arriveTangent=-1.1765, interpMode=RCIM_Cubic, leaveTangent=-1.1765, time=0.796, value=0.245}, <at 1 0>]}}})
    Scale RGB=&VectorFromFloat(Value=&FloatFromCurve(FloatCurve=DataInterface() {propertyValues {Curve {keys [<at 0 1>,<at 0.1242 0.9958 RCIM_Constant>,<at 0.1835 0 RCIM_Constant>,<at 0.284 0.984 RCIM_Constant>,<at 0.5317 0 RCIM_Constant>,<at 0.6402 0.984 RCIM_Constant>,<at 1 0>]}}}))
    ScaleA=true
    ScaleRGB=true
   ) {module=ScaleColor003}
   &DynamicMaterialParameters(
    Index 0 Param 1=&FloatFromCurve(FloatCurve=DataInterface() {propertyValues {Curve {keys [<at 0 1>,<at 0.7 0.35>,<at 1 0>]}}})
    Index 0 Param 2=0.3
    Index 0 Param 3=&FloatFromCurve(FloatCurve=DataInterface() {propertyValues {Curve {keys [<at 0.6 1>,<at 1 0>]}}}, Scale Curve=0.5)
    Index 0 Param 4=1
    Module.Param0WriteEnabled=true
    Module.Param1WriteEnabled=false
    Module.Param2WriteEnabled=true
    Module.Param3WriteEnabled=true
   )
  ]
  rendererClasses [&NiagaraMeshRendererProperties]
  renderers [{rendererClass=&NiagaraMeshRendererProperties, rendererIndex=0}]
 }
}
summary {
 systemName=NS_Hightower_Tapas_Emote
 emitters [
  {emitterName=lightning_burst, rendererClasses [&NiagaraMeshRendererProperties]}
  {emitterName=Lightning_small, rendererClasses [&NiagaraLightRendererProperties, &NiagaraMeshRendererProperties]}
  {emitterName=light, rendererClasses [&NiagaraLightRendererProperties]}
  {emitterName=glow, rendererClasses [&NiagaraSpriteRendererProperties]}
  {emitterName=Lightning_center, rendererClasses [&NiagaraMeshRendererProperties]}
  {emitterName=Lightning_big02, rendererClasses [&NiagaraMeshRendererProperties]}
  {emitterName=Lightning_residue, rendererClasses [&NiagaraMeshRendererProperties]}
  {emitterName=Flare, rendererClasses [&NiagaraSpriteRendererProperties]}
  {emitterName=FlareTop, rendererClasses [&NiagaraSpriteRendererProperties]}
  {emitterName=light_of_big_lightning, rendererClasses [&NiagaraLightRendererProperties]}
  {emitterName=Start_light, rendererClasses [&NiagaraLightRendererProperties]}
  {emitterName=Sparks, rendererClasses [&NiagaraSpriteRendererProperties]}
  {emitterName=Blue_Decal, rendererClasses [&NiagaraSpriteRendererProperties, &NiagaraLightRendererProperties]}
  {
   emitterName=Lightning_small001
   rendererClasses [&NiagaraLightRendererProperties, &NiagaraMeshRendererProperties]
  }
  {
   emitterName=Lightning_small002
   rendererClasses [&NiagaraLightRendererProperties, &NiagaraMeshRendererProperties]
  }
  {emitterName=Hand_Arcs001, rendererClasses [&NiagaraSpriteRendererProperties]}
  {emitterName=lightning_burst001, rendererClasses [&NiagaraMeshRendererProperties]}
 ]
 userVariables [
  User.Skeletal Mesh {
   type=NiagaraDataInterfaceSkeletalMesh
   defaultValue {
    struct=NiagaraExt_VariableValue_DataInterface
    value {dataInterface=&self:NiagaraDataInterfaceSkeletalMesh_0}
   }
  }
  User.SkeletalMesh {
   type=NiagaraDataInterfaceSkeletalMesh
   defaultValue {
    struct=NiagaraExt_VariableValue_DataInterface
    value {dataInterface=&self:NiagaraDataInterfaceSkeletalMesh_2}
   }
  }
  User.Weapon {
   type=NiagaraDataInterfaceSkeletalMesh
   defaultValue {
    struct=NiagaraExt_VariableValue_DataInterface
    value {dataInterface=&self:NiagaraDataInterfaceSkeletalMesh_1}
   }
  }
 ]
}
