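# Niagara effect for the HighTowerTapas emote, showing blue-white meshes at the character's right-hand weapon socket plus a faint blue ribbon trail.
# Four mesh emitters each burst one two-second particle, placed by spawn-time socket sampling and offset, with alpha curves fading it in and out.
# One layer also animates dynamic material parameters, while the short-lived ribbon tapers at the simulation position.
# Emitters derive from skydive contrail templates, with per-frame bone tracking disabled.
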NiagaraSystem /BRCosmetics/Athena/Effects/Particles/Emotes/NS_Emote_HighTowerTapas/P_HighTowe_Tapas_Emote_1
dependencies {usedDataInterfaces [&NiagaraDataInterfaceCurve], usedDynamicInputs [], usedModules [&SystemState]}
emitters {
 FN_Contrails_AttachedParticle_v2 {
  emitterSpawnScript [&FN_PreserveSkydiveUserVariables() {module=PreserveUserVariables}]
  emitterUpdateScript [
   &EmitterState(Inactive Response=::NewEnumerator1, Life Cycle Mode=::NewEnumerator1, Loop Behavior=::NewEnumerator1, Loop Duration=2, UseLoopDelay=false)
   &SpawnBurst_Instantaneous(Spawn Count=1, Use Spawn Probability=false)
  ]
  particleSpawnScript [
   /Game/Effects/Niagara/Modules/Groups/FN_SetSpawnGroup_v1/FN_SetSpawnGroup(Emitter Bone Count=4)
   /Niagara/Modules/Spawn/Initialization/InitializeParticle(
    Color=<rgba 0 0.077 0.1302 0.3>
    Lifetime=2
    Mesh Scale=<0.178 0.178 0.2>
    Module.Write Color=true
    Module.Write Lifetime=true
    Module.Write Mass=true
    Module.Write Material Random=true
    Module.Write Position=false
    Module.Write Scale=true
    Module.Write SpriteRotation=false
    Module.Write SpriteSize=false
    Position=&SimulationPosition()
   )
   /Niagara/Modules/Spawn/Orientation/V2/InitialMeshOrientation(
    Orientation Axis:Vector3f=<0 0 0>
    Orientation Vector=<-20 50 0>
    Rotation=<0 0.3 0>
    Rotation Coordinate Space:ENiagaraMeshTransforms=::NewEnumerator2
    Use Orientation Vector=true
    Use Rotation Vector=true
   )
   &SkeletalMeshLocation(
    Inherit Velocity=false
    MaskSpawnGroup=false
    Mesh Sampling Type=::NewEnumerator5
    Module.Override Randomness=false
    Skeletal Mesh=DataInterface() {propertyValues {
      DefaultMesh=null
      ExcludeBoneName=root
      OptionalComponentClass=null
      PreviewMesh=&M_Med_Soldier_04_Preview
      SkinningMode=SkinOnTheFly
      SoftSourceActor=null
      SourceMode=Default
      UvSetIndex=0
      WholeMeshLOD=-1
      bExcludeBone=false
      bReadDeformedGeometry=true
      bRequireCurrentFrameData=true
      ComponentTags []
      FilteredBones []
      FilteredSockets [weapon_r_socket]
      MeshUserParameter {parameter {name=None, typeDefHandle {registeredTypeIndex=1}, varData []}}
      SamplingRegions []
     }}
   )
   &BoxLocation(Box Offset=<12 0 0>, Box Size=<0 0 0>, MaskSpawnGroup=false, Module.Override Seed=false, Rotate Box=false, Surface Only=false)
  ]
  particleUpdateScript [
   &ParticleState()
   /Niagara/Modules/Update/Orientation/V2/UpdateMeshOrientation()
   &SampleSkeletalMeshSkeleton(Mesh:NiagaraDataInterfaceSkeletalMesh=$User.SkeletalMesh, Specific Bone Index:NiagaraInt32=$Particles.SpawnGroup) {enabled=false}
   /Niagara/Modules/Spawn/Location/SkeletalMeshSkeletonLocation(Module.Inherit Velocity=false, Module.Use Sampled Bone as Ribbon ID=true) {enabled=false}
   &SolveForcesAndVelocity(Clamp Velocity=false, Limit Acceleration=false) {enabled=false}
   &Color(
    Color=<rgba 0.2949 0.4529 0.6823 1>
    Scale Alpha=&FloatFromCurve(FloatCurve=DataInterface() {propertyValues {Curve {keys [
         {arriveTangent=6.0842, interpMode=RCIM_Cubic, leaveTangent=6.0842, tangentMode=RCTM_User, time=0, value=0}
         {arriveTangent=2.2222, interpMode=RCIM_Cubic, leaveTangent=2.2222, time=0.2121, value=1}
         {arriveTangent=-1.2692, interpMode=RCIM_Cubic, leaveTangent=-1.2692, time=0.45, value=1}
         <at 1 0 RCIM_Cubic>
        ]}}})
   )
  ]
  rendererClasses [&NiagaraMeshRendererProperties]
  renderers [{rendererClass=&NiagaraMeshRendererProperties, rendererIndex=0}]
 }
 FN_Contrails_AttachedParticle_v2001 {
  emitterSpawnScript [&FN_PreserveSkydiveUserVariables() {module=PreserveUserVariables}]
  emitterUpdateScript [
   &EmitterState(Inactive Response=::NewEnumerator1, Life Cycle Mode=::NewEnumerator1, Loop Behavior=::NewEnumerator1, Loop Duration=2, UseLoopDelay=false)
   &SpawnBurst_Instantaneous(Spawn Count=1, Use Spawn Probability=false)
  ]
  particleSpawnScript [
   &FN_SetSpawnGroup(Emitter Bone Count=4)
   &InitializeParticle(
    Color=<rgba 0 0.077 0.1302 0.3>
    Lifetime=2
    Mesh Scale=<1.65 1.65 1.65>
    Module.Write Color=true
    Module.Write Lifetime=true
    Module.Write Mass=true
    Module.Write Material Random=true
    Module.Write Position=false
    Module.Write Scale=true
    Module.Write SpriteRotation=false
    Module.Write SpriteSize=false
    Position=&SimulationPosition()
   )
   &InitialMeshOrientation(
    Orientation Axis:Vector3f=<0 0 0>
    Orientation Vector=<0 50 0>
    Rotation=<0.7 0 0>
    Rotation Coordinate Space:ENiagaraMeshTransforms=::NewEnumerator2
    Use Orientation Vector=true
    Use Rotation Vector=true
   )
   &SkeletalMeshLocation(
    Inherit Velocity=false
    MaskSpawnGroup=false
    Mesh Sampling Type=::NewEnumerator5
    Module.Override Randomness=false
    Skeletal Mesh=DataInterface() {propertyValues {
      DefaultMesh=null
      ExcludeBoneName=root
      OptionalComponentClass=null
      PreviewMesh=&M_Med_Soldier_04_Preview
      SkinningMode=SkinOnTheFly
      SoftSourceActor=null
      SourceMode=Default
      UvSetIndex=0
      WholeMeshLOD=-1
      bExcludeBone=false
      bReadDeformedGeometry=true
      bRequireCurrentFrameData=true
      ComponentTags []
      FilteredBones []
      FilteredSockets [weapon_r_socket]
      MeshUserParameter {parameter {name=None, typeDefHandle {registeredTypeIndex=1}, varData []}}
      SamplingRegions []
     }}
   )
   &BoxLocation(Box Offset=<12 0 0>, Box Size=<0 0 0>, MaskSpawnGroup=false, Module.Override Seed=false, Rotate Box=false, Surface Only=false)
   &BoxLocation(Box Offset=<12 0 0>, Box Size=<0 0 0>, MaskSpawnGroup=false, Module.Override Seed=false, Rotate Box=false, Surface Only=false) {module=BoxLocation001}
  ]
  particleUpdateScript [
   &ParticleState()
   &UpdateMeshOrientation()
   &SampleSkeletalMeshSkeleton(Mesh:NiagaraDataInterfaceSkeletalMesh=$User.SkeletalMesh, Specific Bone Index:NiagaraInt32=$Particles.SpawnGroup) {enabled=false}
   &SkeletalMeshSkeletonLocation(Module.Inherit Velocity=false, Module.Use Sampled Bone as Ribbon ID=true) {enabled=false}
   &SolveForcesAndVelocity(Clamp Velocity=false, Limit Acceleration=false) {enabled=false}
   &Color(
    Color=<rgba 0.9271 0.9568 1 1>
    Scale Alpha=&FloatFromCurve(FloatCurve=DataInterface() {propertyValues {Curve {keys [
         {arriveTangent=6.0842, interpMode=RCIM_Cubic, leaveTangent=6.0842, tangentMode=RCTM_User, time=0, value=0}
         {arriveTangent=2.2222, interpMode=RCIM_Cubic, leaveTangent=2.2222, time=0.2121, value=1}
         {arriveTangent=-1.2692, interpMode=RCIM_Cubic, leaveTangent=-1.2692, time=0.45, value=1}
         <at 1 0 RCIM_Cubic>
        ]}}})
   )
   &CameraOffset(Camera Offset Amount=30)
  ]
  rendererClasses [&NiagaraMeshRendererProperties]
  renderers [{rendererClass=&NiagaraMeshRendererProperties, rendererIndex=0}]
 }
 FN_Contrails_AttachedParticle_v2002 {
  emitterSpawnScript [&FN_PreserveSkydiveUserVariables() {module=PreserveUserVariables}]
  emitterUpdateScript [
   &EmitterState(Inactive Response=::NewEnumerator1, Life Cycle Mode=::NewEnumerator1, Loop Behavior=::NewEnumerator1, Loop Duration=2, UseLoopDelay=false)
   &SpawnBurst_Instantaneous(Spawn Count=1, Use Spawn Probability=false)
  ]
  particleSpawnScript [
   &FN_SetSpawnGroup(Emitter Bone Count=4)
   &InitializeParticle(
    Color=<rgba 0 0.077 0.1302 0.3>
    Lifetime=2
    Mesh Scale=<1.7 1.7 1.7>
    Module.Write Color=true
    Module.Write Lifetime=true
    Module.Write Mass=true
    Module.Write Material Random=true
    Module.Write Position=false
    Module.Write Scale=true
    Module.Write SpriteRotation=false
    Module.Write SpriteSize=false
    Position=&SimulationPosition()
   )
   &InitialMeshOrientation(
    Orientation Axis:Vector3f=<0 0 0>
    Orientation Vector=<0 50 0>
    Rotation=<0.7 0 0>
    Rotation Coordinate Space:ENiagaraMeshTransforms=::NewEnumerator2
    Use Orientation Vector=true
    Use Rotation Vector=true
   )
   &SkeletalMeshLocation(
    Inherit Velocity=false
    MaskSpawnGroup=false
    Mesh Sampling Type=::NewEnumerator5
    Module.Override Randomness=false
    Skeletal Mesh=DataInterface() {propertyValues {
      DefaultMesh=null
      ExcludeBoneName=root
      OptionalComponentClass=null
      PreviewMesh=&M_Med_Soldier_04_Preview
      SkinningMode=SkinOnTheFly
      SoftSourceActor=null
      SourceMode=Default
      UvSetIndex=0
      WholeMeshLOD=-1
      bExcludeBone=false
      bReadDeformedGeometry=true
      bRequireCurrentFrameData=true
      ComponentTags []
      FilteredBones []
      FilteredSockets [weapon_r_socket]
      MeshUserParameter {parameter {name=None, typeDefHandle {registeredTypeIndex=1}, varData []}}
      SamplingRegions []
     }}
   )
   &BoxLocation(Box Offset=<20 0 0>, Box Size=<0 0 0>, MaskSpawnGroup=false, Module.Override Seed=false, Rotate Box=false, Surface Only=false)
  ]
  particleUpdateScript [
   &ParticleState()
   &UpdateMeshOrientation()
   &SampleSkeletalMeshSkeleton(Mesh:NiagaraDataInterfaceSkeletalMesh=$User.SkeletalMesh, Specific Bone Index:NiagaraInt32=$Particles.SpawnGroup) {enabled=false}
   &SkeletalMeshSkeletonLocation(Module.Inherit Velocity=false, Module.Use Sampled Bone as Ribbon ID=true) {enabled=false}
   &SolveForcesAndVelocity(Clamp Velocity=false, Limit Acceleration=false) {enabled=false}
   &Color(
    Color=<rgba 0.1348 0.489 0.9583 1>
    Scale Alpha=&FloatFromCurve(FloatCurve=DataInterface() {propertyValues {Curve {keys [
         {arriveTangent=6.0842, interpMode=RCIM_Cubic, leaveTangent=6.0842, tangentMode=RCTM_User, time=0, value=0}
         {arriveTangent=2.2222, interpMode=RCIM_Cubic, leaveTangent=2.2222, time=0.2121, value=1}
         {arriveTangent=-1.2692, interpMode=RCIM_Cubic, leaveTangent=-1.2692, time=0.45, value=1}
         <at 1 0 RCIM_Cubic>
        ]}}})
   )
   &CameraOffset(Camera Offset Amount=10)
  ]
  rendererClasses [&NiagaraMeshRendererProperties]
  renderers [{rendererClass=&NiagaraMeshRendererProperties, rendererIndex=0}]
 }
 FN_Contrails_AttachedParticle_v2003 {
  emitterSpawnScript [&FN_PreserveSkydiveUserVariables() {module=PreserveUserVariables}]
  emitterUpdateScript [
   &EmitterState(Inactive Response=::NewEnumerator1, Life Cycle Mode=::NewEnumerator1, Loop Behavior=::NewEnumerator1, Loop Duration=2, UseLoopDelay=false)
   &SpawnBurst_Instantaneous(Spawn Count=1, Use Spawn Probability=false)
  ]
  particleSpawnScript [
   &FN_SetSpawnGroup(Emitter Bone Count=4)
   &InitializeParticle(
    Color=<rgba 0 0.077 0.1302 0.3>
    Lifetime=2
    Mesh Scale=<0.75 0.75 0.75>
    Module.Write Color=true
    Module.Write Lifetime=true
    Module.Write Mass=true
    Module.Write Material Random=true
    Module.Write Position=false
    Module.Write Scale=true
    Module.Write SpriteRotation=false
    Module.Write SpriteSize=false
    Position=&SimulationPosition()
   )
   &InitialMeshOrientation(
    Orientation Axis:Vector3f=<0 0 0>
    Orientation Vector=<0 50 0>
    Rotation Coordinate Space:ENiagaraMeshTransforms=::NewEnumerator2
    Use Orientation Vector=true
    Use Rotation Vector=true
   )
   &SkeletalMeshLocation(
    Inherit Velocity=false
    MaskSpawnGroup=false
    Mesh Sampling Type=::NewEnumerator5
    Module.Override Randomness=false
    Skeletal Mesh=DataInterface() {propertyValues {
      DefaultMesh=null
      ExcludeBoneName=root
      OptionalComponentClass=null
      PreviewMesh=&M_Med_Soldier_04_Preview
      SkinningMode=SkinOnTheFly
      SoftSourceActor=null
      SourceMode=Default
      UvSetIndex=0
      WholeMeshLOD=-1
      bExcludeBone=false
      bReadDeformedGeometry=true
      bRequireCurrentFrameData=true
      ComponentTags []
      FilteredBones []
      FilteredSockets [weapon_r_socket]
      MeshUserParameter {parameter {name=None, typeDefHandle {registeredTypeIndex=1}, varData []}}
      SamplingRegions []
     }}
   )
   &BoxLocation(Box Offset=<12 0 0>, Box Size=<0 0 0>, MaskSpawnGroup=false, Module.Override Seed=false, Rotate Box=false, Surface Only=false)
  ]
  particleUpdateScript [
   &ParticleState()
   &UpdateMeshOrientation()
   &SampleSkeletalMeshSkeleton(Mesh:NiagaraDataInterfaceSkeletalMesh=$User.SkeletalMesh, Specific Bone Index:NiagaraInt32=$Particles.SpawnGroup) {enabled=false}
   &SkeletalMeshSkeletonLocation(Module.Inherit Velocity=false, Module.Use Sampled Bone as Ribbon ID=true) {enabled=false}
   &SolveForcesAndVelocity(Clamp Velocity=false, Limit Acceleration=false) {enabled=false}
   &Color(
    Color=<rgba 0.9271 0.9568 1 1.9>
    Scale Alpha=&FloatFromCurve(FloatCurve=DataInterface() {propertyValues {Curve {keys [
         {arriveTangent=6.0842, interpMode=RCIM_Cubic, leaveTangent=6.0842, tangentMode=RCTM_User, time=0, value=0}
         {arriveTangent=1.6894, interpMode=RCIM_Cubic, leaveTangent=1.6894, time=0.194, value=0.8638}
         {arriveTangent=-1.0718, interpMode=RCIM_Cubic, leaveTangent=-1.0718, time=0.5607, value=0.9472}
         <at 1 0 RCIM_Cubic>
        ]}}})
   )
   &CameraOffset(Camera Offset Amount=10)
   &DynamicMaterialParameters(
    Index 0 Param 1=1
    Index 0 Param 2=&FloatFromCurve(FloatCurve=DataInterface() {propertyValues {Curve {keys [<at 0.15 0>,<at 0.3 1>,<at 1 0>]}}})
    Index 0 Param 3=&FloatFromCurve(FloatCurve=DataInterface() {propertyValues {Curve {keys [<at 0.35 0>,<at 0.65 1>,<at 1 0>]}}})
    Module.Param0WriteEnabled=true
    Module.Param1WriteEnabled=true
    Module.Param2WriteEnabled=true
    Module.Param3WriteEnabled=true
   )
  ]
  rendererClasses [&NiagaraMeshRendererProperties]
  renderers [{rendererClass=&NiagaraMeshRendererProperties, rendererIndex=0}]
 }
 FN_RibbonTrail {
  emitterUpdateScript [
   &EmitterScalability(
    Module.bApplySpawnCountScale=false
    Module.bApplySpawnCountScaleByDistanceFraction=false
    Module.bUseMaxDistance=false
    Module.bUseMinDistance=false
   ) {enabled=false}
   &EmitterLifeCycle() {enabled=false}
   &EmitterState()
   &SpawnRate(SpawnRate=20, Use Spawn Probability=false)
  ]
  particleSpawnScript [
   &InitializeParticle(
    Color=<rgba 0.2407 0.6594 1 0.03>
    Lifetime=0.37
    Module.Write Color=true
    Module.Write Lifetime=true
    Module.Write Mass=true
    Module.Write Material Random=true
    Module.Write Position=false
    Module.Write Scale=false
    Module.Write SpriteRotation=false
    Module.Write SpriteSize=false
    Position=&SimulationPosition()
   )
   /Niagara/Modules/Ribbons/FN_InitializeRibbonAttributes(Ribbon Width=35)
   /Game/Effects/Niagara/Modules/Sprite/FN_InitializeSpriteAttributes(Module.Write SpriteRotation=false, Module.Write SpriteSize=false)
  ]
  particleUpdateScript [
   &ParticleState()
   &ScaleRibbonWidth(Ribbon Width Scale=&FloatFromCurve(FloatCurve=DataInterface() {propertyValues {Curve {keys [<at -0.0046 1.0113>,<at 1.0011 0.002>]}}}))
  ]
  rendererClasses [&NiagaraRibbonRendererProperties]
  renderers [{rendererClass=&NiagaraRibbonRendererProperties}]
 }
}
summary {
 systemName=P_HighTowe_Tapas_Emote_1
 emitters [
  {emitterName=FN_Contrails_AttachedParticle_v2, rendererClasses [&NiagaraMeshRendererProperties]}
  {emitterName=FN_RibbonTrail, rendererClasses [&NiagaraRibbonRendererProperties]}
  {emitterName=FN_Contrails_AttachedParticle_v2001, rendererClasses [&NiagaraMeshRendererProperties]}
  {emitterName=FN_Contrails_AttachedParticle_v2002, rendererClasses [&NiagaraMeshRendererProperties]}
  {emitterName=FN_Contrails_AttachedParticle_v2003, rendererClasses [&NiagaraMeshRendererProperties]}
 ]
 userVariables [
  User.bIsFrontEndPreview {type=NiagaraBool}
  User.PlayerSpeed {type=NiagaraFloat, defaultValue=0, description=""}
  User.RibbonIDOffset {type=NiagaraInt32}
  User.SkeletalMesh {
   type=NiagaraDataInterfaceSkeletalMesh
   defaultValue {
    struct=NiagaraExt_VariableValue_DataInterface
    value {dataInterface=&self:NiagaraDataInterfaceSkeletalMesh_0}
   }
  }
 ]
}
